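# Physics asset for the female medium RangeGawker cosmetic body, covering the dynamic bones of its jacket.
# Capsules on spine_02 to spine_04 form the torso, and small sphere MSTR bodies root each jacket chain to spine_03 with free angular constraints.
# Front, side and back jacket capsules hang from those roots with tight swing and twist limits.
# The front chains have two segments, and the left adds a twist-locked zipper pull.
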PhysicsAsset /BRCosmetics/Characters/Player/Female/Medium/Bodies/F_MED_RangeGawker/Meshes/F_MED_RangeGawker_Physics
bodies {
 dyn_jacket_bk_01 [{shapeType=Capsule, length=5.208, radius=2.5784, shapeName=dyn_jacket_bk_01_capsule, center=<3.1633 0.0501 0.8471>, rotation=<rot 0 0 -90>}]
 dyn_jacket_bk_MSTR [{shapeType=Sphere, radius=0.505, shapeName=dyn_jacket_bk_MSTR_sphere}]
 dyn_jacket_bk_l_01 [{
   shapeType=Capsule
   length=2.886
   radius=2.4362
   shapeName=dyn_jacket_bk_l_01_capsule
   center=<2.8222 0.0984 0.1989>
   rotation=<rot 90 -7.125 171.8849>
  }]
 dyn_jacket_bk_l_MSTR [{shapeType=Sphere, radius=0.505, shapeName=dyn_jacket_bk_l_MSTR_sphere}]
 dyn_jacket_bk_r_01 [{
   shapeType=Capsule
   length=2.886
   radius=2.4362
   shapeName=dyn_jacket_bk_r_01_capsule
   center=<-2.826 0.1225 -0.1989>
   rotation=<rot 90 -166.8094 -167.7994>
  }]
 dyn_jacket_bk_r_MSTR [{shapeType=Sphere, radius=0.505, shapeName=dyn_jacket_bk_r_MSTR_sphere}]
 dyn_jacket_fr_l_01 [{
   shapeType=Capsule
   length=3.3666
   radius=2.6263
   shapeName=dyn_jacket_fr_l_02_capsule
   center=<2.2706 -0.1138 -0.1769>
   rotation=<rot 0 0.0001 87.5671>
  }]
 dyn_jacket_fr_l_02 [{
   shapeType=Capsule
   length=3.3666
   radius=2.6263
   shapeName=dyn_jacket_fr_l_02_capsule
   center=<2.2706 -0.1138 -0.1769>
   rotation=<rot 0 0.0001 87.5671>
  }]
 dyn_jacket_fr_l_MSTR [{shapeType=Sphere, radius=0.505, shapeName=dyn_jacket_fr_l_MSTR_sphere}]
 dyn_jacket_fr_l_zipper [{
   shapeType=Capsule
   length=1.1189
   radius=0.8485
   shapeName=dyn_jacket_fr_l_zipper_capsule
   center=<1.0243 0.0183 -0.0016>
   rotation=<rot 90 0 0>
  }]
 dyn_jacket_fr_r_01 [{
   shapeType=Capsule
   length=3.3666
   radius=2.6263
   shapeName=dyn_jacket_fr_r_02_capsule
   center=<-2.2706 0.1138 0.1769>
   rotation=<rot 0 0.0001 -92.4329>
  }]
 dyn_jacket_fr_r_02 [{
   shapeType=Capsule
   length=3.3666
   radius=2.6263
   shapeName=dyn_jacket_fr_r_02_capsule
   center=<-2.2706 0.1138 0.1769>
   rotation=<rot 0 0.0001 -92.4329>
  }]
 dyn_jacket_fr_r_MSTR [{shapeType=Sphere, radius=0.505, shapeName=dyn_jacket_fr_r_MSTR_sphere}]
 dyn_jacket_l_01 [{
   shapeType=Capsule
   length=3.2301
   radius=2.8052
   shapeName=dyn_jacket_l_01_capsule
   center=<3.7395 -0.4312 0.0101>
   rotation=<rot 90 179.7016 2.9051>
  }]
 dyn_jacket_l_MSTR [{shapeType=Sphere, radius=0.505, shapeName=dyn_jacket_l_MSTR_sphere}]
 dyn_jacket_r_01 [{
   shapeType=Capsule
   length=3.2301
   radius=2.8052
   shapeName=dyn_jacket_r_01_capsule
   center=<-3.7395 0.4312 -0.0101>
   rotation=<rot 90 180 -176.7964>
  }]
 dyn_jacket_r_MSTR [{shapeType=Sphere, radius=0.505, shapeName=dyn_jacket_r_MSTR_sphere}]
 spine_02 [{
   shapeType=Capsule
   length=1.0815
   radius=6.797
   shapeName=spine_02_capsule
   center=<1.0861 1.8841 0.0439>
   rotation=<rot -1.7817 -84.6376 0.1321>
  }]
 spine_03 [{
   shapeType=Capsule
   length=2.7464
   radius=6.1581
   shapeName=spine_03_capsule
   center=<1.6273 2.3649 -0.0695>
   rotation=<rot 0.1381 -82.479 -0.3274>
  }]
 spine_04 [{shapeType=Capsule, length=7.0966, radius=4.4127, shapeName=spine_04_capsule, center=<-0.1025 2.9173 -0.021>, rotation=<rot 3.1915 -90 0.989>}]
}
constraints [
 {bone1Name=dyn_jacket_fr_l_02, bone2Name=dyn_jacket_fr_l_01, swing1LimitDegrees=10, swing2LimitDegrees=25, twistLimitDegrees=5}
 {
  bone1Name=dyn_jacket_fr_l_zipper
  bone2Name=dyn_jacket_fr_l_02
  swing1LimitDegrees=15
  swing2LimitDegrees=45
  twistLimitDegrees=5
  twistMotion=Locked
 }
 {bone1Name=dyn_jacket_fr_r_02, bone2Name=dyn_jacket_fr_r_01, swing1LimitDegrees=10, swing2LimitDegrees=25, twistLimitDegrees=5}
 {bone1Name=dyn_jacket_bk_l_01, bone2Name=dyn_jacket_bk_l_MSTR, swing1LimitDegrees=5, swing2LimitDegrees=25, twistLimitDegrees=5}
 {bone1Name=dyn_jacket_fr_r_01, bone2Name=dyn_jacket_fr_r_MSTR, swing1LimitDegrees=10, swing2LimitDegrees=25, twistLimitDegrees=5}
 {bone1Name=dyn_jacket_fr_l_01, bone2Name=dyn_jacket_fr_l_MSTR, swing1LimitDegrees=10, swing2LimitDegrees=25, twistLimitDegrees=5}
 {bone1Name=dyn_jacket_l_01, bone2Name=dyn_jacket_l_MSTR, swing1LimitDegrees=10, swing2LimitDegrees=25, twistLimitDegrees=5}
 {bone1Name=dyn_jacket_r_01, bone2Name=dyn_jacket_r_MSTR, swing1LimitDegrees=10, swing2LimitDegrees=25, twistLimitDegrees=5}
 {bone1Name=dyn_jacket_bk_r_01, bone2Name=dyn_jacket_bk_r_MSTR, swing1LimitDegrees=5, swing2LimitDegrees=25, twistLimitDegrees=5}
 {bone1Name=dyn_jacket_bk_01, bone2Name=dyn_jacket_bk_MSTR, swing1LimitDegrees=5, swing2LimitDegrees=25, twistLimitDegrees=5}
 {
  bone1Name=spine_03
  bone2Name=spine_02
  swing1LimitDegrees=45
  swing1Motion=Free
  swing2LimitDegrees=45
  swing2Motion=Free
  twistLimitDegrees=45
  twistMotion=Free
 }
 {
  bone1Name=spine_04
  bone2Name=spine_03
  swing1LimitDegrees=45
  swing1Motion=Free
  swing2LimitDegrees=45
  swing2Motion=Free
  twistLimitDegrees=45
  twistMotion=Free
 }
 {
  bone1Name=dyn_jacket_r_MSTR
  bone2Name=spine_03
  swing1LimitDegrees=45
  swing1Motion=Free
  swing2LimitDegrees=45
  swing2Motion=Free
  twistLimitDegrees=45
  twistMotion=Free
 }
 {
  bone1Name=dyn_jacket_l_MSTR
  bone2Name=spine_03
  swing1LimitDegrees=45
  swing1Motion=Free
  swing2LimitDegrees=45
  swing2Motion=Free
  twistLimitDegrees=45
  twistMotion=Free
 }
 {
  bone1Name=dyn_jacket_fr_r_MSTR
  bone2Name=spine_03
  swing1LimitDegrees=45
  swing1Motion=Free
  swing2LimitDegrees=45
  swing2Motion=Free
  twistLimitDegrees=45
  twistMotion=Free
 }
 {
  bone1Name=dyn_jacket_fr_l_MSTR
  bone2Name=spine_03
  swing1LimitDegrees=45
  swing1Motion=Free
  swing2LimitDegrees=45
  swing2Motion=Free
  twistLimitDegrees=45
  twistMotion=Free
 }
 {
  bone1Name=dyn_jacket_bk_r_MSTR
  bone2Name=spine_03
  swing1LimitDegrees=45
  swing1Motion=Free
  swing2LimitDegrees=45
  swing2Motion=Free
  twistLimitDegrees=45
  twistMotion=Free
 }
 {
  bone1Name=dyn_jacket_bk_l_MSTR
  bone2Name=spine_03
  swing1LimitDegrees=45
  swing1Motion=Free
  swing2LimitDegrees=45
  swing2Motion=Free
  twistLimitDegrees=45
  twistMotion=Free
 }
 {
  bone1Name=dyn_jacket_bk_MSTR
  bone2Name=spine_03
  swing1LimitDegrees=45
  swing1Motion=Free
  swing2LimitDegrees=45
  swing2Motion=Free
  twistLimitDegrees=45
  twistMotion=Free
 }
]
