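# Physics asset for the face accessory mesh of the female medium RangeGawker cosmetic, defining collision and joints for dynamic hair bones.
# Capsules cover the head and four hair bones per side.
# Bones 01 to 03 form a chain from the head on each side, while each 04 bone constrains directly to the head.
# Swing1 stays narrow throughout, while swing2 widens toward chain tips, more on the right than the left.
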PhysicsAsset /BRCosmetics/Characters/Player/Female/Medium/Bodies/F_MED_RangeGawker/Meshes/Parts/F_MED_RangeGawker_FaceAcc_Physics
bodies {
 dyn_hair_l_01 [{
   shapeType=Capsule
   length=2.8583
   radius=1.5561
   shapeName=dyn_hair_l_01_capsule
   center=<-0.0221 0.1014 -1.162>
   rotation=<rot -0.0023 9.8042 -87.1659>
  }]
 dyn_hair_l_02 [{
   shapeType=Capsule
   length=2.4268
   radius=1.4552
   shapeName=dyn_hair_l_02_capsule
   center=<-0.0047 -0.0883 -0.6913>
   rotation=<rot -0.2905 14.3438 -89.0854>
  }]
 dyn_hair_l_03 [{
   shapeType=Capsule
   length=2.487
   radius=1.4946
   shapeName=dyn_hair_l_03_capsule
   center=<0.1326 -0.0348 -1.0446>
   rotation=<rot -0.3792 8.0759 -88.0596>
  }]
 dyn_hair_l_04 [{
   shapeType=Capsule
   length=0.1427
   radius=2.9689
   shapeName=dyn_hair_l_04_capsule
   center=<0.0198 0.1442 2.3524>
   rotation=<rot -2.2025 112.5868 6.0506>
  }]
 dyn_hair_r_01 [{
   shapeType=Capsule
   length=2.9302
   radius=1.4845
   shapeName=dyn_hair_r_01_capsule
   center=<0.2648 0.0383 -1.2417>
   rotation=<rot 0 -1.3814 -82.5113>
  }]
 dyn_hair_r_02 [{
   shapeType=Capsule
   length=1.4417
   radius=1.5283
   shapeName=dyn_hair_r_02_capsule
   center=<0.3296 0.1325 -0.957>
   rotation=<rot 0 -3.9226 -86.0673>
  }]
 dyn_hair_r_03 [{
   shapeType=Capsule
   length=1.7775
   radius=1.396
   shapeName=dyn_hair_r_03_capsule
   center=<0.0038 -0.4566 -1.0267>
   rotation=<rot 30.4929 -95.1138 17.5048>
  }]
 dyn_hair_r_04 [{
   shapeType=Capsule
   length=0.1427
   radius=2.9689
   shapeName=dyn_hair_r_04_capsule
   center=<-0.1866 -0.1829 2.2928>
   rotation=<rot -2.2025 112.5868 -173.9494>
  }]
 head [{
   shapeType=Capsule
   length=6.0511
   radius=5.4631
   shapeName=head_capsule
   center=<2.9859 3.2668 -0.0009>
   rotation=<rot -0.5235 -90.0002 90.3342>
  }]
}
constraints [
 {bone1Name=dyn_hair_r_03, bone2Name=dyn_hair_r_02, swing1LimitDegrees=5, swing2LimitDegrees=55, twistLimitDegrees=15}
 {bone1Name=dyn_hair_r_02, bone2Name=dyn_hair_r_01, swing1LimitDegrees=5, swing2LimitDegrees=20, twistLimitDegrees=10}
 {bone1Name=dyn_hair_r_01, bone2Name=head, swing1LimitDegrees=5, swing2LimitDegrees=10, twistLimitDegrees=5}
 {bone1Name=dyn_hair_l_04, bone2Name=head, swing1LimitDegrees=5, swing2LimitDegrees=15, twistLimitDegrees=15}
 {bone1Name=dyn_hair_l_03, bone2Name=dyn_hair_l_02, swing1LimitDegrees=5, swing2LimitDegrees=35, twistLimitDegrees=15}
 {bone1Name=dyn_hair_l_02, bone2Name=dyn_hair_l_01, swing1LimitDegrees=5, swing2LimitDegrees=20, twistLimitDegrees=10}
 {bone1Name=dyn_hair_l_01, bone2Name=head, swing1LimitDegrees=5, swing2LimitDegrees=5, twistLimitDegrees=5}
 {bone1Name=dyn_hair_r_04, bone2Name=head, swing1LimitDegrees=5, swing2LimitDegrees=15, twistLimitDegrees=15}
]
